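# Sound cue for the indoor room ambient loop in the Apollo MountainBase biome.
# A single looping wave player feeds the output directly, playing a Mole-named bridge room ambience wave.
# It uses Athena ambient attenuation and the indoor ambient sound class, with slightly raised pitch, reduced volume and a very large max distance.
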SoundCue /Game/Sounds/Apollo/Biomes/MountainBase/MountainBase_AmbientLoop_Room_Cue
root=WavePlayer_0
nodes [{
  id=WavePlayer_0
  properties {bLooping=true, soundWaveAssetPtr=/Game/Sounds/Apollo/Biomes/MountainBase/MoleWaves/Mole_AmbientLoop_Room_Bridge_01}
 }]
properties {
 attenuationSettings=&Fort_Ambient_Athena_Att
 maxDistance=2097152
 pitchMultiplier=1.1
 soundClassObject=&Fort_Ambient_Indoors_Athena
 volumeMultiplier=0.697
}
