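# Sound cue for non-looping indoor ambient one-shots in the MountainBase biome.
# A Random node picks with equal weight among hallway ambience, access-point motor and interior metal waves.
# A root Modulator varies pitch slightly and randomizes volume on each play to reduce repetition.
# The cue uses the indoor ambient sound class, shared ambient attenuation and a reduced volume multiplier.
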SoundCue /Game/Sounds/Apollo/Biomes/MountainBase/MountainBase_Ambient_NonLooping_Cue
root=Modulator_0
nodes [
 {id=Modulator_0, children [Random_0], properties {pitchMax=1.02, pitchMin=0.98, volumeMin=0.7}}
 {
  id=Random_0
  children [
   WavePlayer_0
   WavePlayer_1
   WavePlayer_2
   WavePlayer_3
   WavePlayer_4
   WavePlayer_5
   WavePlayer_6
   WavePlayer_7
   WavePlayer_8
   WavePlayer_9
   WavePlayer_10
   WavePlayer_11
   WavePlayer_12
   WavePlayer_13
   WavePlayer_14
   WavePlayer_15
   WavePlayer_16
   WavePlayer_17
   WavePlayer_18
  ]
  properties {weights [1,1,1,1,1,1,1,1,1,1,1,1,1,1,1,1,1,1,1]}
 }
 {id=WavePlayer_0, properties {soundWaveAssetPtr=/Game/Sounds/Apollo/Biomes/MountainBase/MoleWaves/Mole_Ambient_NonLooping_Hallway_01}}
 {id=WavePlayer_1, properties {soundWaveAssetPtr=/Game/Sounds/Apollo/Biomes/MountainBase/MoleWaves/Mole_Ambient_NonLooping_Hallway_02}}
 {id=WavePlayer_10, properties {soundWaveAssetPtr=/Game/Sounds/Apollo/Biomes/MountainBase/MoleWaves/Biome_Access_Point_MotorRunMed_01}}
 {id=WavePlayer_11, properties {soundWaveAssetPtr=/Game/Sounds/Apollo/Biomes/MountainBase/MoleWaves/Biome_Access_Point_MotorRunMed_02}}
 {id=WavePlayer_12, properties {soundWaveAssetPtr=/Game/Sounds/Apollo/Biomes/MountainBase/MoleWaves/Biome_Access_Point_MotorRunMed_03}}
 {id=WavePlayer_13, properties {soundWaveAssetPtr=/Game/Sounds/Apollo/Biomes/MountainBase/MoleWaves/Biome_Access_Point_MotorRunMed_04}}
 {id=WavePlayer_14, properties {soundWaveAssetPtr=/Game/Sounds/Apollo/Biomes/MountainBase/MoleWaves/AMB_SS_Int_Metal_01}}
 {id=WavePlayer_15, properties {soundWaveAssetPtr=/Game/Sounds/Apollo/Biomes/MountainBase/MoleWaves/AMB_SS_Int_Metal_04}}
 {id=WavePlayer_16, properties {soundWaveAssetPtr=/Game/Sounds/Apollo/Biomes/MountainBase/MoleWaves/AMB_SS_Int_Metal_06}}
 {id=WavePlayer_17, properties {soundWaveAssetPtr=/Game/Sounds/Apollo/Biomes/MountainBase/MoleWaves/AMB_SS_Int_Metal_10}}
 {id=WavePlayer_18, properties {soundWaveAssetPtr=/Game/Sounds/Apollo/Biomes/MountainBase/MoleWaves/AMB_SS_Int_Metal_12}}
 {id=WavePlayer_2, properties {soundWaveAssetPtr=/Game/Sounds/Apollo/Biomes/MountainBase/MoleWaves/Mole_Ambient_NonLooping_Hallway_03}}
 {id=WavePlayer_3, properties {soundWaveAssetPtr=/Game/Sounds/Apollo/Biomes/MountainBase/MoleWaves/Mole_Ambient_NonLooping_Hallway_04}}
 {id=WavePlayer_4, properties {soundWaveAssetPtr=/Game/Sounds/Apollo/Biomes/MountainBase/MoleWaves/Mole_Ambient_NonLooping_Hallway_05}}
 {id=WavePlayer_5, properties {soundWaveAssetPtr=/Game/Sounds/Apollo/Biomes/MountainBase/MoleWaves/Mole_Ambient_NonLooping_Hallway_06}}
 {id=WavePlayer_6, properties {soundWaveAssetPtr=/Game/Sounds/Apollo/Biomes/MountainBase/MoleWaves/Mole_Ambient_NonLooping_Hallway_07}}
 {id=WavePlayer_7, properties {soundWaveAssetPtr=/Game/Sounds/Apollo/Biomes/MountainBase/MoleWaves/Mole_Ambient_NonLooping_Hallway_08}}
 {id=WavePlayer_8, properties {soundWaveAssetPtr=/Game/Sounds/Apollo/Biomes/MountainBase/MoleWaves/Mole_Ambient_NonLooping_Hallway_09}}
 {id=WavePlayer_9, properties {soundWaveAssetPtr=/Game/Sounds/Apollo/Biomes/MountainBase/MoleWaves/Mole_Ambient_NonLooping_Hallway_10}}
]
properties {
 attenuationSettings=&Fort_Ambient_Athena_Att
 maxDistance=2097152
 soundClassObject=&Fort_Ambient_Indoors_Athena
 volumeMultiplier=0.644
 preEffectBusSends [{}]
}
