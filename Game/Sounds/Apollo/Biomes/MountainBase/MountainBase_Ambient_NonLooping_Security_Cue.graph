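# Sound cue for non-looping ambient one-shots in the MountainBase biome, named for its security ambience.
# A Random node picks with equal weight among twelve waves: energy releases, distant trapped-tower sounds and computer blips.
# A Modulator at the root adds slight pitch variation and varies volume on each play.
# The cue uses the indoor ambient sound class and shared ambient attenuation, at reduced volume.
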SoundCue /Game/Sounds/Apollo/Biomes/MountainBase/MountainBase_Ambient_NonLooping_Security_Cue
root=Modulator_1
nodes [
 {id=Modulator_1, children [Random_1], properties {pitchMax=1.02, pitchMin=0.98, volumeMin=0.75}}
 {
  id=Random_1
  children [
   WavePlayer_19
   WavePlayer_20
   WavePlayer_21
   WavePlayer_22
   WavePlayer_26
   WavePlayer_27
   WavePlayer_28
   WavePlayer_29
   WavePlayer_30
   WavePlayer_31
   WavePlayer_32
   WavePlayer_33
  ]
  properties {weights [1,1,1,1,1,1,1,1,1,1,1,1]}
 }
 {id=WavePlayer_19, properties {soundWaveAssetPtr=/Game/Sounds/Apollo/Biomes/MountainBase/MoleWaves/AMB_SS_S15_ZP_EnergyRelease_02}}
 {id=WavePlayer_20, properties {soundWaveAssetPtr=/Game/Sounds/Apollo/Biomes/MountainBase/MoleWaves/AMB_SS_S15_ZP_EnergyRelease_05}}
 {id=WavePlayer_21, properties {soundWaveAssetPtr=/Game/Sounds/Apollo/Biomes/MountainBase/MoleWaves/AMB_SS_S15_ZP_EnergyRelease_06}}
 {id=WavePlayer_22, properties {soundWaveAssetPtr=/Game/Sounds/Apollo/Biomes/MountainBase/MoleWaves/AMB_SS_S15_ZP_EnergyRelease_07}}
 {id=WavePlayer_26, properties {soundWaveAssetPtr=/Game/Sounds/Apollo/Biomes/MountainBase/MoleWaves/AMB_SS_S16_ZP_Trapped_Tower_Dist_05}}
 {id=WavePlayer_27, properties {soundWaveAssetPtr=/Game/Sounds/Apollo/Biomes/MountainBase/MoleWaves/AMB_SS_S16_ZP_Trapped_Tower_Dist_07}}
 {id=WavePlayer_28, properties {soundWaveAssetPtr=/Game/Sounds/Apollo/Biomes/MountainBase/MoleWaves/AMB_SS_S16_ZP_Trapped_Tower_Dist_06}}
 {id=WavePlayer_29, properties {soundWaveAssetPtr=/Game/Sounds/Apollo/Biomes/MountainBase/MoleWaves/AMB_SS_S15_HuntersHaven_Computer_Blips_02}}
 {id=WavePlayer_30, properties {soundWaveAssetPtr=/Game/Sounds/Apollo/Biomes/MountainBase/MoleWaves/AMB_SS_S15_HuntersHaven_Computer_Blips_06}}
 {id=WavePlayer_31, properties {soundWaveAssetPtr=/Game/Sounds/Apollo/Biomes/MountainBase/MoleWaves/AMB_SS_S15_HuntersHaven_Computer_Blips_07}}
 {id=WavePlayer_32, properties {soundWaveAssetPtr=/Game/Sounds/Apollo/Biomes/MountainBase/MoleWaves/AMB_SS_S15_HuntersHaven_Computer_Blips_10}}
 {id=WavePlayer_33, properties {soundWaveAssetPtr=/Game/Sounds/Apollo/Biomes/MountainBase/MoleWaves/AMB_SS_S15_HuntersHaven_Computer_Blips_11}}
]
properties {
 attenuationSettings=&Fort_Ambient_Athena_Att
 maxDistance=2097152
 soundClassObject=&Fort_Ambient_Indoors_Athena
 volumeMultiplier=0.644
 preEffectBusSends [{}]
}
